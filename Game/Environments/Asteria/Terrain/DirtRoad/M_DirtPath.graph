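# Dirt path material for spline-mesh roads in Asteria terrain, writing surface attributes and displacement to a Runtime Virtual Texture.
# Tiled diffuse, normal and packed SDR maps drive the surface, with SDR green supplying height.
# Opacity height-lerps that height against a UV cross-width falloff with optional, invertible end fade, then breaks it up with world-space Perlin noise.
# MaterialProxyReplace separates realtime from proxy attributes, and mobile RVT base color is desaturated.
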Material /Game/Environments/Asteria/Terrain/DirtRoad/M_DirtPath
graph {
 properties {
  bUseMaterialAttributes=true
  bUsedWithNanite=true
  bUsedWithSplineMeshes=true
  pixelDepthOffsetMode=PDOM_Legacy
  usedShadingModels=Default Lit
 }
 nodes [
  n1 = ScalarParameter() {parameterName=TransitionEnd_1}
  n2 = ScalarParameter() {defaultValue=1, parameterName=TransitionEnd_2}
  n3 = ComponentMask(None=TextureCoordinate() {coordinateIndex=1}) {r=true}
  n4 = StaticSwitchParameter(True=OneMinus(None=@n3), False=@n3) {parameterName=InvertEnd}
  n5 = ComponentMask(None=TextureCoordinate()) {r=true}
  n6 = Multiply(A=@n5, B=OneMinus(None=@n5))
  n7 = Multiply(A=Saturate(None=LinearInterpolate(A=@n1, B=@n2, Alpha=@n4)), B=@n6)
  n8 = LinearInterpolate(
   A=ScalarParameter() {defaultValue=-0.2, parameterName=TransitionRVT_1}
   B=ScalarParameter() {defaultValue=1.89, parameterName=TransitionRVT_2}
   Alpha=Multiply(A=StaticSwitchParameter(True=@n7, False=@n6) {parameterName=IsEnd}) {constB=4}
  )
  n9 = Divide(A=TextureCoordinate(), B=AppendVector(A=Constant() {r=1}, B=ScalarParameter() {defaultValue=1, parameterName=RoadTiling}))
  n10 = TextureSampleParameter2D(UVs=@n9) {parameterName=SDR, texture=&/DirtPath_SDR}
  n11 = WorldPosition()
  n12 = ScalarParameter() {defaultValue=2520.9541, parameterName=WorldNoise_Size}
  n13 = LinearInterpolate(
   A=ScalarParameter() {defaultValue=-0.5, parameterName=WorldNoise_Low}
   B=ScalarParameter() {defaultValue=2, parameterName=WorldNoise_High}
   Alpha=TextureSample(UVs=Divide(A=@n11.XY, B=@n12)) {samplerType=SAMPLERTYPE_LinearGrayscale, texture=&T_SimplePerlin_Greyscale}.R
  )
  n14 = Multiply(
   A=Multiply(
    A=MaterialFunctionCall(Transition Phase=@n8, Height Texture=@n10.G) {materialFunction=&HeightLerp}.Alpha
    B=Saturate(None=@n13)
   )
   B=ScalarParameter() {defaultValue=0.7, desc=Default is 0.7, parameterName=OpacityMod}
  ) {constB=0.7}
  LinearInterpolate(A=@n14, Alpha=ScalarParameter() {parameterName=DEBUG_Opacity})
  n15 = ScalarParameter() {defaultValue=1, parameterName=NormalStrength}
  n16 = TextureSampleParameter2D(UVs=@n9) {parameterName=Normal, samplerType=SAMPLERTYPE_Normal, texture=&/DirtPath_N}
  n17 = ComponentMask(None=@n16.RGB) {g=true, r=true}
  n18 = LinearInterpolate(
   A=ScalarParameter() {group=Displacement, parameterName=Height_DisplacementLow}
   B=ScalarParameter() {defaultValue=1, group=Displacement, parameterName=Height_DisplacementHigh}
   Alpha=@n10.G
  )
  n19 = SetMaterialAttributes!(
   MaterialAttributes
   Base Color=Multiply(
    A=VectorParameter() {parameterName=Color, defaultValue=<rgba 0.7028 0.8125 0.6092 1>}.RGB
    B=TextureSampleParameter2D(UVs=@n9) {parameterName=Diffuse, texture=&/DirtPath_D}.RGB
   )
   Specular=Multiply(A=@n10.R, B=ScalarParameter() {defaultValue=0.5, parameterName=Spec})
   Roughness=Multiply(A=@n10.B, B=ScalarParameter() {defaultValue=1.1, parameterName=Roughness})
   Normal=AppendVector(A=Multiply(A=@n15, B=@n17), B=@n16.B)
   Opacity=@n14
   Displacement=LinearInterpolate(B=Saturate(None=@n18), Alpha=ScalarParameter() {defaultValue=1, group=Displacement, parameterName=Height_DisplacementAmt}) {constA=0.5}
  ) {attributeSetTypes [
    #1
    #2
    #3
    #4
    #5
    #6
   ]}
  MaterialProxyReplace(Realtime=SetMaterialAttributes!(MaterialAttributes), MaterialProxy=@n19)
  n20 = ScalarParameter() {defaultValue=0.15, parameterName=Mobile Desaturation}
  n21 = GetMaterialAttributes(None=@n19) {attributeGetTypes [
    #1
    #2
    #3
    #4
    #5
    #6
   ]}
  RuntimeVirtualTextureOutput!(
   BaseColor=ShadingPathSwitch(Default=@n21.Base Color, Mobile=Desaturation(None=@n21.Base Color, Fraction=@n20))
   Specular=@n21.Specular
   Roughness=@n21.Specular
   Normal=Transform(None=@n21.Normal)
   WorldHeight
   Opacity=@n21.Specular
   Mask=Constant() {r=1}
   Displacement=@n21.Specular
   Mask4
  )
  n22 = ScalarParameter() {defaultValue=1, group=Displacement, parameterName=Height_DisplacementPower}
  Saturate(None=Multiply(A=Power(Base=@n10.G, Exp=@n22), B=ScalarParameter() {defaultValue=1, group=Displacement, parameterName=Height_DisplacementMod}))
 ]
}
