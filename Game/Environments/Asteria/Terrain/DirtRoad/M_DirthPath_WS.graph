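# Dirt path material for spline meshes that feeds a Runtime Virtual Texture output, with extra desaturation on mobile.
# Base textures use world-aligned or scaled-UV projection, with an optional second UV-based layer blended by a contrast-adjusted mask.
# Opacity comes from a UV edge falloff with optional end fade, height-lerped against a texture channel and broken up by world-space Perlin noise.
# Tint, desaturation, normal strength and displacement are adjustable.
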Material /Game/Environments/Asteria/Terrain/DirtRoad/M_DirthPath_WS
graph {
 properties {
  bUseMaterialAttributes=true
  bUsedWithNanite=true
  bUsedWithSplineMeshes=true
  pixelDepthOffsetMode=PDOM_Legacy
  usedShadingModels=Default Lit
 }
 nodes [
  n1 = ScalarParameter() {parameterName=TransitionEnd_1}
  n2 = ScalarParameter() {defaultValue=1, parameterName=TransitionEnd_2}
  n3 = ComponentMask(None=TextureCoordinate() {coordinateIndex=1}) {r=true}
  n4 = StaticSwitchParameter(True=OneMinus(None=@n3), False=@n3) {parameterName=InvertEnd}
  n5 = ComponentMask(None=TextureCoordinate()) {r=true}
  n6 = Multiply(A=@n5, B=OneMinus(None=@n5))
  n7 = Multiply(A=Saturate(None=LinearInterpolate(A=@n1, B=@n2, Alpha=@n4)), B=@n6)
  n8 = LinearInterpolate(
   A=ScalarParameter() {defaultValue=-0.2, parameterName=TransitionRVT_1}
   B=ScalarParameter() {defaultValue=1.89, parameterName=TransitionRVT_2}
   Alpha=Multiply(A=StaticSwitchParameter(True=@n7, False=@n6) {parameterName=IsEnd}) {constB=4}
  )
  n9 = MaterialFunctionCall(Color=NamedRerouteUsage()) {materialFunction=&SplitComponents}
  n10 = Divide(A=WorldPosition().XY, B=ScalarParameter() {defaultValue=2520.9541, parameterName=WorldNoise_Size})
  n11 = MaterialFunctionCall(
   In=LinearInterpolate(
    A=ScalarParameter() {defaultValue=-0.5, parameterName=WorldNoise_Low}
    B=ScalarParameter() {defaultValue=2, parameterName=WorldNoise_High}
    Alpha=TextureSample(UVs=@n10) {samplerType=SAMPLERTYPE_LinearGrayscale, texture=&T_SimplePerlin_Greyscale}.R
   )
   Contrast=ScalarParameter() {parameterName=Contrast amount}
  ) {materialFunction=&CheapContrast}
  n12 = Multiply(
   A=Multiply(
    A=MaterialFunctionCall(Transition Phase=@n8, Height Texture=@n9.G) {materialFunction=&HeightLerp}.Alpha
    B=Saturate(None=@n11.Result)
   )
   B=ScalarParameter() {defaultValue=0.7, desc=Default is 0.7, parameterName=OpacityMod}
  ) {constB=0.7}
  LinearInterpolate(A=@n12, Alpha=ScalarParameter() {parameterName=DEBUG_Opacity})
  n13 = NamedRerouteUsage()
  n14 = ScalarParameter() {parameterName=Desat amount}
  n15 = ScalarParameter() {defaultValue=1, parameterName=NormalStrength}
  n16 = MaterialFunctionCall(Color=NamedRerouteUsage()) {materialFunction=&SplitComponents}
  n17 = ComponentMask(None=@n16.RGB) {g=true, r=true}
  n18 = LinearInterpolate(
   A=ScalarParameter() {group=Displacement, parameterName=Height_DisplacementLow}
   B=ScalarParameter() {defaultValue=1, group=Displacement, parameterName=Height_DisplacementHigh}
   Alpha=NamedRerouteUsage()
  )
  n19 = SetMaterialAttributes!(
   MaterialAttributes
   Base Color=Multiply(A=VectorParameter() {parameterName=Color, defaultValue=<rgba 0.7028 0.8125 0.6092 1>}.RGB, B=Desaturation(None=@n13, Fraction=@n14))
   Specular=Multiply(A=@n9.R, B=ScalarParameter() {defaultValue=0.5, parameterName=Spec})
   Roughness=Multiply(A=@n9.B, B=ScalarParameter() {defaultValue=1.1, parameterName=Roughness})
   Normal=AppendVector(A=Multiply(A=@n15, B=@n17), B=@n16.B)
   Opacity=@n12
   Displacement=LinearInterpolate(B=Saturate(None=@n18), Alpha=ScalarParameter() {defaultValue=1, group=Displacement, parameterName=Height_DisplacementAmt}) {constA=0.5}
  ) {attributeSetTypes [
    #1
    #2
    #3
    #4
    #5
    #6
   ]}
  MaterialProxyReplace(Realtime=SetMaterialAttributes!(MaterialAttributes), MaterialProxy=@n19)
  n20 = MaterialFunctionCall(Float2=TextureCoordinate()) {materialFunction=&BreakOutFloat2Components}
  n21 = MaterialFunctionCall(X=@n20.R, Y=Multiply(A=@n20.G, B=ScalarParameter() {defaultValue=1, group=Textures_L2, parameterName=L2_ScaleTextures})) {materialFunction=&MakeFloat2}
  n22 = TextureSample(
   UVs=@n21.Result
   Tex=TextureObjectParameter() {group=Textures_L2, parameterName=Diffuse_L2, texture=/Game/Environments/Apollo/Sets/Terrain/Textures/T_Apollo_Cliff_D_02}
  )
  n23 = TextureObjectParameter() {
   group=Textures_L2
   parameterName=Specular_L2
   samplerType=SAMPLERTYPE_Masks
   texture=/Game/Environments/Apollo/Sets/Terrain/Textures/T_Apollo_Cliff_S_02
  }
  n24 = TextureSample(UVs=@n21.Result, Tex=@n23) {samplerType=SAMPLERTYPE_Masks}
  n25 = TextureObjectParameter() {
   group=Textures_L2
   parameterName=Normal_L2
   samplerType=SAMPLERTYPE_Normal
   texture=/Game/Environments/Apollo/Sets/Terrain/Textures/T_Apollo_Cliff_N_02
  }
  n26 = TextureSample(UVs=@n21.Result, Tex=@n25) {samplerType=SAMPLERTYPE_Normal}
  n27 = NamedRerouteUsage()
  n28 = MaterialFunctionCall(Color=NamedRerouteUsage()) {materialFunction=&SplitComponents}
  n29 = NamedRerouteUsage()
  n30 = NamedRerouteUsage()
  n31 = Power(Base=OneMinus(None=@n30), Exp=ScalarParameter() {defaultValue=4, group=Textures_L2, parameterName=L2_MaskContrast})
  n32 = Power(Base=@n30, Exp=ScalarParameter() {defaultValue=4, group=Textures_L2, parameterName=L2_MaskContrast})
  n33 = BreakMaterialAttributes!(Attr=BlendMaterialAttributes(
    A=MakeMaterialAttributes!(
     BaseColor=@n22.RGB
     Metallic=@n24.G
     Specular=@n24.G
     Roughness=@n24.G
     Anisotropy
     EmissiveColor
     Opacity
     OpacityMask
     Normal=@n26.RGB
     Tangent
     WorldPositionOffset
     SubsurfaceColor
     ClearCoat
     ClearCoatRoughness
     AmbientOcclusion
     Refraction
     CustomizedUVs_0
     CustomizedUVs_1
     CustomizedUVs_2
     CustomizedUVs_3
     CustomizedUVs_4
     CustomizedUVs_5
     CustomizedUVs_6
     CustomizedUVs_7
     PixelDepthOffset
     ShadingModel
     Displacement=@n22.RGB
    )
    B=MakeMaterialAttributes(BaseColor=@n27, Metallic=@n28.G, Specular=@n28.G, Roughness=@n28.G, Normal=@n29, Displacement=@n30)
    Alpha=StaticSwitchParameter(True=@n31, False=@n32) {parameterName=L2_FlipMask}
   ))
  NamedRerouteDeclaration(None=@n33.BaseColor) {name=L2_Diffuse, nodeColor=<rgba 1 0 0.0234 1>}
  n34 = MaterialFunctionCall(Texture Scale=ScalarParameter() {defaultValue=1000, group=Textures, parameterName=ScaleTextures}) {materialFunction=&ScaleUVsByCenter}
  n35 = TextureObjectParameter() {group=Textures, parameterName=Diffuse, texture=&T_Apollo_Cliff_D_02}
  n36 = ScalarParameter() {defaultValue=1000, group=Textures, parameterName=ScaleTextures}
  n37 = StaticSwitchParameter(
   True=TextureSample(UVs=@n34.UVs, Tex=@n35).RGBA
   False=MaterialFunctionCall(TextureObject=@n35, TextureSize=@n36, Export Float 4=StaticBool() {value=true}) {materialFunction=&WorldAlignedTexture}.Z Texture
  ) {parameterName=UseUVs}
  n38 = ComponentMask(None=@n37) {b=true, g=true, r=true}
  NamedRerouteDeclaration(None=@n38) {name=L1_Diffuse, nodeColor=<rgba 1 0.375 0 1>}
  n39 = ComponentMask(None=@n37) {a=true}
  NamedRerouteDeclaration(None=@n39) {name=L1_Height, nodeColor=<rgba 0.0547 0 1 1>}
  n40 = TextureObjectParameter() {
   group=Textures
   parameterName=Normal
   samplerType=SAMPLERTYPE_Normal
   texture=&T_Apollo_Cliff_N_02
  }
  n41 = StaticSwitchParameter(
   True=TextureSample(UVs=@n34.UVs, Tex=@n40) {samplerType=SAMPLERTYPE_Normal}.RGB
   False=MaterialFunctionCall(TextureObject=@n40, TextureSize=@n36) {materialFunction=&WorldAlignedNormal}.Z Texture
  ) {parameterName=UseUVs}
  NamedRerouteDeclaration(None=@n41) {name=L1_Normals, nodeColor=<rgba 0.2656 1 0 1>}
  n42 = TextureObjectParameter() {
   group=Textures
   parameterName=Specular
   samplerType=SAMPLERTYPE_Masks
   texture=&T_Apollo_Cliff_S_02
  }
  n43 = StaticSwitchParameter(
   True=TextureSample(UVs=@n34.UVs, Tex=@n42) {samplerType=SAMPLERTYPE_Masks}.RGB
   False=MaterialFunctionCall(TextureObject=@n42, TextureSize=@n36) {materialFunction=&WorldAlignedTexture}.Z Texture
  ) {parameterName=UseUVs}
  NamedRerouteDeclaration(None=StaticSwitchParameter(True=NamedRerouteUsage(), False=@n43) {desc=Enable to add a second UV based layer, parameterName=Use2layer}) {name=Spec, nodeColor=<rgba 0 0.6953 1 1>}
  NamedRerouteDeclaration(None=@n33.Displacement) {name=L2_Height, nodeColor=<rgba 0 0.3437 1 1>}
  NamedRerouteDeclaration(None=StaticSwitchParameter(True=NamedRerouteUsage(), False=@n41) {desc=Enable to add a second UV based layer, parameterName=Use2layer}) {name=Normals, nodeColor=<rgba 0 0.6953 1 1>}
  NamedRerouteDeclaration(None=StaticSwitchParameter(True=NamedRerouteUsage(), False=@n39) {desc=Enable to add a second UV based layer, parameterName=Use2layer}) {name=DiffuseHeight, nodeColor=<rgba 0 0.6953 1 1>}
  NamedRerouteDeclaration(None=StaticSwitchParameter(True=NamedRerouteUsage(), False=@n38) {desc=Enable to add a second UV based layer, parameterName=Use2layer}) {name=Diffuse, nodeColor=<rgba 0 0.6953 1 1>}
  NamedRerouteDeclaration(None=@n33.Normal) {name=L2_Normal, nodeColor=<rgba 0.6641 1 0 1>}
  NamedRerouteDeclaration(None=MaterialFunctionCall(X=@n33.Specular, Y=@n33.Specular, Z=@n33.Specular) {materialFunction=&MakeFloat3}.Result) {name=L2_Spec, nodeColor=<rgba 1 0 0.9844 1>}
  NamedRerouteDeclaration(None=@n43) {name=L1_Spec, nodeColor=<rgba 0 1 0.6953 1>}
  n44 = ScalarParameter() {defaultValue=0.15, parameterName=Mobile Desaturation}
  n45 = GetMaterialAttributes(None=@n19) {attributeGetTypes [
    #1
    #2
    #3
    #4
    #5
    #6
   ]}
  RuntimeVirtualTextureOutput!(
   BaseColor=ShadingPathSwitch(Default=@n45.Base Color, Mobile=Desaturation(None=@n45.Base Color, Fraction=@n44))
   Specular=@n45.Specular
   Roughness=@n45.Specular
   Normal=Transform(None=@n45.Normal)
   WorldHeight
   Opacity=@n45.Specular
   Mask=Constant() {r=1}
   Displacement=@n45.Specular
   Mask4
  )
  n46 = ScalarParameter() {defaultValue=1, group=Displacement, parameterName=Height_DisplacementPower}
  Saturate(None=Multiply(A=Power(Base=@n9.G, Exp=@n46), B=ScalarParameter() {defaultValue=1, group=Displacement, parameterName=Height_DisplacementMod}))
 ]
}
