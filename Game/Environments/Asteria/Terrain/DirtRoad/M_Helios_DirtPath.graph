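# Dirt road material for Nanite and spline meshes in the Asteria terrain, writing its dirt layer to a Runtime Virtual Texture.
# Tiled dirt textures get noise-driven colour variation and height-blended opacity from UV side fades, world-space noise and an optional invertible end fade.
# A moss layer is height-blended over the dirt using world noise and side masks, but the RVT output reads only dirt.
# RVT colour has a mobile desaturate-and-multiply path, plus switchable displacement.
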Material /Game/Environments/Asteria/Terrain/DirtRoad/M_Helios_DirtPath
graph {
 properties {
  bTangentSpaceNormal=false
  bUseMaterialAttributes=true
  bUsedWithNanite=true
  bUsedWithSplineMeshes=true
  pixelDepthOffsetMode=PDOM_Legacy
  usedShadingModels=Default Lit
 }
 nodes [
  n1 = VectorParameter() {group=Textures, parameterName=Color overlay, defaultValue=<rgba 0.7028 0.8125 0.6092 1>}
  n2 = Divide(A=TextureCoordinate(), B=AppendVector(A=Constant() {r=1}, B=ScalarParameter() {defaultValue=1, parameterName=RoadTiling}))
  n3 = TextureSampleParameter2D(UVs=@n2) {group=Textures, parameterName=Diffuse, texture=&/DirtPath_D}
  n4 = TextureSampleParameter2D(UVs=@n2) {group=Textures, parameterName=SDR, texture=&/DirtPath_SDR}
  n5 = TextureSampleParameter2D(UVs=@n2) {group=Textures, parameterName=Normal, samplerType=SAMPLERTYPE_Normal, texture=&/DirtPath_N}
  n6 = TextureSampleParameter2D(UVs=Divide(A=WorldPosition().XY, B=ScalarParameter() {defaultValue=2520.9541, group=Opacity mask, parameterName=Opacity mask size})) {
   group=Opacity mask
   parameterName=Opacity mask texture
   samplerType=SAMPLERTYPE_LinearGrayscale
   texture=&T_SimplePerlin_Greyscale
  }
  n7 = Saturate(None=LinearInterpolate(
    A=ScalarParameter() {defaultValue=-0.5, group=Opacity mask, parameterName=Opacity mask in}
    B=ScalarParameter() {defaultValue=2, group=Opacity mask, parameterName=Opacity mask out}
    Alpha=@n6.R
   ))
  n8 = TextureCoordinate()
  n9 = Multiply(A=@n8, B=OneMinus(None=ComponentMask(None=@n8) {r=true}))
  n10 = ComponentMask(None=TextureCoordinate() {coordinateIndex=1}) {r=true}
  n11 = LinearInterpolate(
   A=ScalarParameter() {parameterName=TransitionEnd_1}
   B=ScalarParameter() {defaultValue=1, parameterName=TransitionEnd_2}
   Alpha=StaticSwitchParameter(True=OneMinus(None=@n10), False=@n10) {parameterName=InvertEnd}
  )
  n12 = StaticSwitchParameter(True=Multiply(A=@n9, B=Saturate(None=@n11)), False=@n9) {parameterName=IsEnd}
  n13 = MaterialFunctionCall(
   Transition Phase=LinearInterpolate(
    A=ScalarParameter() {defaultValue=-0.2051, parameterName=RVT in}
    B=ScalarParameter() {defaultValue=1.8894, parameterName=RVT out}
    Alpha=Multiply(A=@n12) {constB=4}
   )
   Height Texture=@n4.G
  ) {materialFunction=&HeightLerp}
  n14 = SetMaterialAttributes!(
   MaterialAttributes
   Base Color=LinearInterpolate(
    A=Multiply(A=@n1.RGB, B=@n3.RGB)
    B=VectorParameter() {group=Color variation, parameterName=Color variation, defaultValue=<rgba 1 0 0 1>}.RGB
    Alpha=NamedRerouteUsage()
   )
   Specular=Multiply(A=@n4.R, B=ScalarParameter() {defaultValue=0.5, group=Textures, parameterName=Spec})
   Roughness=Multiply(A=@n4.B, B=ScalarParameter() {defaultValue=1.1, group=Textures, parameterName=Roughness})
   Normal=@n5.RGB
   Opacity=Multiply(A=Multiply(A=@n7, B=@n13.Alpha), B=ScalarParameter() {defaultValue=0.7, group=Opacity mask, parameterName=Opacity mask softness}) {constB=0.1}
  ) {attributeSetTypes [
    #1
    #2
    #3
    #4
    #5
   ]}
  n15 = Divide(
   A=TextureCoordinate()
   B=AppendVector(A=ScalarParameter() {defaultValue=0.22, parameterName=GrassSizeX}, B=ScalarParameter() {defaultValue=1, parameterName=GrassSizeY})
  )
  n16 = TextureSample(UVs=@n15) {samplerType=SAMPLERTYPE_Masks, texture=/Game/Environments/Asteria/Sets/ZenGarden/Textures/T_Asteria_Zen_MossyGround_01_S}
  n17 = SetMaterialAttributes!(
   MaterialAttributes
   Base Color=Multiply(
    A=TextureSample(UVs=@n15) {texture=/Game/Environments/Asteria/Foliage/ForestFloor/T_RainForest_JungleMoss_D}.RGB
    B=VectorParameter() {parameterName=MossColor, defaultValue=<rgba 2 0.9089 0.2978 1>}.RGB
   )
   Metallic
   Specular=@n16.R
   Roughness=@n16.R
   Normal=TextureSample(UVs=@n15) {samplerType=SAMPLERTYPE_Normal, texture=/Game/Environments/Asteria/Sets/ZenGarden/Textures/T_Asteria_Zen_MossyGround_01_N}.RGB
  ) {attributeSetTypes [
    #1
    #6
    #2
    #3
    #4
   ]}
  n18 = Divide(A=WorldPosition().XY, B=ScalarParameter() {defaultValue=2423.6099, parameterName=WorldNoiseSize_Grass})
  n19 = Saturate(None=LinearInterpolate(
    A=ScalarParameter() {defaultValue=0.6391, parameterName=Grasstransition_1}
    B=ScalarParameter() {defaultValue=1.4351, parameterName=Grasstransition_2}
    Alpha=TextureSample(UVs=@n18) {samplerType=SAMPLERTYPE_LinearGrayscale, texture=&T_SimplePerlin_Greyscale}.R
   ))
  n20 = TextureCoordinate()
  n21 = ComponentMask(None=@n20) {r=true}
  n22 = Multiply(A=Multiply(A=@n20, B=OneMinus(None=@n21))) {constB=4}
  n23 = LinearInterpolate(
   A=ScalarParameter() {defaultValue=-1.6115, parameterName=Sidetransition_1}
   B=ScalarParameter() {defaultValue=0.8932, parameterName=Sidetransition_2}
   Alpha=@n22
  )
  n24 = Multiply(A=@n4.G, B=Saturate(None=@n23))
  n25 = TextureSample(UVs=@n15) {texture=/Game/Environments/Asteria/Sets/ZenGarden/Textures/T_Asteria_Zen_MossyGround_01_H}
  n26 = ScalarParameter() {defaultValue=0.304, parameterName=2nHeight_1}
  BlendMaterialAttributes(
   A=@n14
   B=@n17
   Alpha=MaterialFunctionCall(
    Transition Phase=Multiply(A=@n19, B=@n24)
    Height Texture=Add(A=@n25.R, B=@n26)
    Contrast=ScalarParameter() {defaultValue=2.1096, parameterName=2nHeight_Contrast}
   ) {materialFunction=&HeightLerp}.Alpha
  )
  Constant()
  MaterialProxyReplace(Realtime=SetMaterialAttributes!(MaterialAttributes), MaterialProxy=@n14)
  Multiply(B=@n26)
  Multiply(A=@n26, B=OneMinus(None=@n25.R))
  n27 = ScalarParameter() {defaultValue=-0.5, group=Color variation, parameterName=Color variation noise in}
  n28 = ScalarParameter() {defaultValue=2, group=Color variation, parameterName=Color variation noise out}
  n29 = TextureSampleParameter2D(UVs=Divide(A=WorldPosition().XY, B=ScalarParameter() {defaultValue=2000, group=Color variation, parameterName=Color variation noise size})) {
   group=Color variation
   parameterName=Color variation noise texture
   samplerType=SAMPLERTYPE_LinearGrayscale
   texture=&T_SimplePerlin_Greyscale
  }
  NamedRerouteDeclaration(None=Saturate(None=LinearInterpolate(A=@n27, B=@n28, Alpha=@n29.R))) {name=Color variation mask, nodeColor=<rgba 1 0 0.3984 1>}
  n30 = Power(Base=@n3.A, Exp=ScalarParameter() {defaultValue=1, group=Displacement, parameterName=Heightmap_Power})
  n31 = ScalarParameter() {defaultValue=1, group=Displacement, parameterName=Heightmap_Mod}
  NamedRerouteDeclaration(None=Saturate(None=Multiply(A=@n30, B=@n31))) {name=Heightmap, nodeColor=<rgba 1 0 0.3984 1>}
  n32 = GetMaterialAttributes(None=@n14) {attributeGetTypes [
    #1
    #2
    #3
    #4
    #5
   ]}
  n33 = Desaturation(None=@n32.Base Color, Fraction=ScalarParameter() {parameterName=Mobile Desaturation})
  n34 = ScalarParameter() {defaultValue=0.7, parameterName=Mobile Multiplier}
  RuntimeVirtualTextureOutput!(
   BaseColor=ShadingPathSwitch(Default=@n32.Base Color, Mobile=Multiply(A=@n33, B=@n34))
   Specular=@n32.Specular
   Roughness=@n32.Specular
   Normal=Transform(None=@n32.Normal)
   WorldHeight
   Opacity=@n32.Specular
   Mask=Constant() {r=1}
   Displacement=StaticSwitch(True=NamedRerouteUsage(), False=Constant(), Value=StaticBoolParameter() {group=Displacement, parameterName=UseDisplacement})
   Mask4
  )
  n35 = Add(A=@n19, B=@n23)
  n36 = ScalarParameter() {parameterName=GrassContrast}
  n37 = LinearInterpolate(A=@n26, B=ScalarParameter() {defaultValue=-236.4354, parameterName=2nHeight_2}, Alpha=@n25.R)
  Saturate(None=Add(
    A=MaterialFunctionCall(Transition Phase=@n35, Height Texture=@n22, Contrast=@n36) {materialFunction=&HeightLerp}.Alpha
    B=Saturate(None=@n37)
   ))
 ]
}
